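# Physics asset for the HicT1 vehicle body skeletal mesh, likely the battle royale variant given the BR suffix.
# It approximates the whole vehicle with a single capsule on the body bone, laid along its long axis and slightly offset from the origin.
# It has no constraints, so it provides one simple collision volume rather than an articulated ragdoll.
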PhysicsAsset /VehicleCosmetics/Bodies/HicT1/Meshes/Body_HicT1_PhysicsBR
bodies {body [{shapeType=Capsule, length=263.5317, radius=122.8646, center=<39.0122 -1.2826 37.039>, rotation=<rot 0 -89.8538 -88.4261>}]}
constraints []
